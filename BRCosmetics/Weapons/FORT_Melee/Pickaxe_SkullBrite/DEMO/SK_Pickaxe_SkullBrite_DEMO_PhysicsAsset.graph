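# Physics asset for the SkullBrite demo pickaxe skeletal mesh, a two-handed melee cosmetic weapon.
# A single capsule body on the Melee_TwoHanded_Handle bone provides its simplified collision shape.
# The capsule is offset and rotated from the bone, and the asset has no constraints, so there are no articulated parts.
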PhysicsAsset /BRCosmetics/Weapons/FORT_Melee/Pickaxe_SkullBrite/DEMO/SK_Pickaxe_SkullBrite_DEMO_PhysicsAsset
bodies {Melee_TwoHanded_Handle [{shapeType=Capsule, length=62.0711, radius=31.3979, center=<0.7603 3.7841 21.4319>, rotation=<rot 1.6844 -90 0.2337>}]}
constraints []
